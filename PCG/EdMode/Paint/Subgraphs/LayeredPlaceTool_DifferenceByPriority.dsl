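; PCG loop subgraph for the layered place tool in paint mode that trims each layer's points against accumulated layers with lower Priority values.
; RemainingPoints acts as loop feedback: the first iteration passes SourcePoints through, later ones turn data tags into attributes to compare Priority.
; Matching data indices are joined to select the data for a binary Difference on the source, skipped when none match.
; Non-empty results are gathered with the unchanged accumulated points.
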(graph LayeredPlaceTool_DifferenceByPriority
  (inputs (SourcePoints :type Point) (RemainingPoints :type Point))
  (outputs (RemainingPoints :type Point))
  (bind GetLoopIndex_0 (Get%20Loop%20Index))
  (bind Branch_38 (Branch))
  (bind TagsToAttributeSet_0 (Data%20Tags%20To%20Attribute%20Set))
  (bind MergeAttributes_1 (Merge%20Attributes))
  (bind CopyAttributes_0 (Copy%20Attributes :InputSource $Index :OutputTarget DataIndex))
  (bind Branch_0 (Branch))
  (bind TagsToAttributeSet_13 (Data%20Tags%20To%20Attribute%20Set))
  (bind AttributeFilter_11 (Filter%20Attribute%20Elements :AttributeTypes (struct :DoubleValue 1.0) :Operator Lesser :TargetAttribute Priority :ThresholdAttribute Priority))
  (bind RemoveEmptyData_0 (Remove%20Empty%20Data))
  (bind DataNum_1 (Data%20Count))
  (bind Branch_1 (Branch :bOutputToB true))
  (bind AttributeReduce_0 (Attribute%20Reduce :InputSource DataIndex :JoinDelimiter "," :Operation Join :OutputAttributeName SelectedIndices))
  (bind Branch_5 (Branch :bOutputToB true))
  (bind FilterDataByIndex_0 (Filter%20Data%20By%20Index))
  (bind ToPoint_24 (To%20Point))
  (bind RemoveEmptyData_1 (Remove%20Empty%20Data))
  (bind Gather_0 (Gather))
  ;[0 0 0 1 32] Difference
  (bind Difference_22 (Difference :bDiffMetadata false :DensityFunction Binary :Mode Discrete))
  (connect TagsToAttributeSet_13 Out AttributeFilter_11 Filter)
  (connect CopyAttributes_0 Out AttributeFilter_11 In)
  (connect RemoveEmptyData_0 Out AttributeReduce_0 In)
  (connect GetLoopIndex_0 Out Branch_0 bOutputToB)
  (connect DefaultInputNode SourcePoints Branch_0 In)
  (connect DataNum_1 Out Branch_1 bOutputToB)
  (connect Branch_0 Output%20B Branch_1 In)
  (connect DataNum_1 Out Branch_5 bOutputToB)
  (connect Branch_38 Output%20B Branch_5 In)
  (connect GetLoopIndex_0 Out Branch_38 bOutputToB)
  (connect DefaultInputNode RemainingPoints Branch_38 In)
  (connect MergeAttributes_1 Out CopyAttributes_0 Source)
  (connect MergeAttributes_1 Out CopyAttributes_0 Target)
  (connect RemoveEmptyData_0 Out DataNum_1 In)
  (connect FilterDataByIndex_0 InsideFilter Difference_22 Differences)
  (connect Branch_1 Output%20B Difference_22 Source)
  (connect Branch_5 Output%20B FilterDataByIndex_0 In)
  (connect AttributeReduce_0 Out FilterDataByIndex_0 SelectedIndices)
  (connect Branch_0 Output%20A Gather_0 In)
  (connect Branch_38 Output%20B Gather_0 In2)
  (connect Branch_1 Output%20A Gather_0 In3)
  (connect RemoveEmptyData_1 Out Gather_0 In4)
  (connect TagsToAttributeSet_0 Out MergeAttributes_1 In)
  (connect AttributeFilter_11 InsideFilter RemoveEmptyData_0 In)
  (connect ToPoint_24 Out RemoveEmptyData_1 In)
  (connect Branch_38 Output%20B TagsToAttributeSet_0 In)
  (connect Branch_0 Output%20B TagsToAttributeSet_13 In)
  (connect Difference_22 Out ToPoint_24 In)
  (output RemainingPoints Gather_0))
