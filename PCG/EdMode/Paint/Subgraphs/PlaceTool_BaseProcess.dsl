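; PCG subgraph for the editor paint mode's place tool, turning tool points into tagged, transformed placement points for static meshes or actor classes.
; SpawnActors selects the source list; non-empty lists are weight-matched onto points, while empty ones tag points ShowPreviewMesh.
; Mesh points optionally take bounds from their mesh.
; All paths receive global, then seeded random, transform adjustments, and finally layer Name and Priority tags.
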(graph PlaceTool_BaseProcess
  (params (LayerName :type String) (LayerPriority :type Int) (UseMeshBounds :type Bool) (SpawnActors :type Bool) (GlobalPositionOffset :type Vector) (GlobalRotation :type Rotator) (AbsoluteGlobalRotation :type Bool) (GlobalScaleFactor :type Vector :default (1.0 1.0 1.0)) (RandomPositionOffsetMin :type Vector) (RandomPositionOffsetMax :type Vector) (RandomRotationMin :type Rotator) (RandomRotationMax :type Rotator) (RandomScaleFactorMin :type Vector :default (1.0 1.0 1.0)) (RandomScaleFactorMax :type Vector :default (1.0 1.0 1.0)) (UniformRandomScaleFactor :type Bool :default true) (RandomSeed :type Int :default 2118139241))
  (inputs (ToolData :type Point) (StaticMeshes :type Param) (ActorClasses :type Param))
  (bind Select_1 (Select :bUseInputB-param SpawnActors))
  (bind RemoveEmptyData_1 (Remove%20Empty%20Data))
  (bind DataNum_1 (Data%20Count))
  (bind Branch_0 (Branch))
  (bind MatchAndSetAttributes_1 (Match%20And%20Set%20Attributes :bUseWeightAttribute true :WeightAttribute Weight))
  (bind AddTags_0 (Add%20Tags :TagsToAdd "ShowPreviewMesh"))
  (bind Branch_2 (Branch :bOutputToB-param SpawnActors))
  (bind Branch_1 (Branch :bOutputToB-param UseMeshBounds))
  (bind BoundsFromMesh_0 (Bounds%20From%20Mesh :MeshAttribute StaticMesh))
  (bind TransformPoints_1 (Transform%20Points :bUniformScale false :bAbsoluteRotation-param AbsoluteGlobalRotation :OffsetMax-param GlobalPositionOffset :OffsetMin-param GlobalPositionOffset :RotationMax-param GlobalRotation :RotationMin-param GlobalRotation :ScaleMax-param GlobalScaleFactor :ScaleMin-param GlobalScaleFactor))
  (bind TransformPoints_2 (Transform%20Points :bUniformScale-param UniformRandomScaleFactor :OffsetMax-param RandomPositionOffsetMax :OffsetMin-param RandomPositionOffsetMin :RotationMax-param RandomRotationMax :RotationMin-param RandomRotationMin :ScaleMax-param RandomScaleFactorMax :ScaleMin-param RandomScaleFactorMin :Seed-param RandomSeed))
  (bind AddTags_1 (Add%20Tags :Prefix "Name:" :TagsToAdd-param LayerName))
  (bind AddTags_2 (Add%20Tags :Prefix "Priority:" :TagsToAdd-param LayerPriority))
  (connect Branch_0 Output%20A AddTags_0 In)
  (connect TransformPoints_2 Out AddTags_1 In)
  (connect AddTags_1 Out AddTags_2 In)
  (connect Branch_1 Output%20B BoundsFromMesh_0 In)
  (connect DataNum_1 Out Branch_0 bOutputToB)
  (connect DefaultInputNode ToolData Branch_0 In)
  (connect Branch_2 Output%20A Branch_1 In)
  (connect MatchAndSetAttributes_1 Out Branch_2 In)
  (connect RemoveEmptyData_1 Out DataNum_1 In)
  (connect Branch_0 Output%20B MatchAndSetAttributes_1 In)
  (connect Select_1 Out MatchAndSetAttributes_1 Match%20Data)
  (connect Select_1 Out RemoveEmptyData_1 In)
  (connect DefaultInputNode StaticMeshes Select_1 Input%20A)
  (connect DefaultInputNode ActorClasses Select_1 Input%20B)
  (connect AddTags_0 Out TransformPoints_1 In)
  (connect BoundsFromMesh_0 Out TransformPoints_1 In)
  (connect Branch_1 Output%20A TransformPoints_1 In)
  (connect Branch_2 Output%20B TransformPoints_1 In)
  (connect TransformPoints_1 Out TransformPoints_2 In)
  (output Out AddTags_2))
